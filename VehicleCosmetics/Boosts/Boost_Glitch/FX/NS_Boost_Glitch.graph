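# Niagara vehicle cosmetic effect for the Glitch boost, emitted from the car's filtered skeletal-mesh sockets.
# Four emitters layer ribbon trails, short burst mesh/sprite flares, SubUV sprite particles and a persistent mesh glow.
# The particle emitter merges three spawn groups, selecting size, lifetime, velocity and drag per group.
# Trails and particles cycle red-green-blue unless painted; paint-variant, game-mode and frontend-preview inputs adjust color and motion.
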NiagaraSystem /VehicleCosmetics/Boosts/Boost_Glitch/FX/NS_Boost_Glitch
dependencies {
 usedDataInterfaces [
  &NiagaraDataInterfaceCurve
  &NiagaraDataInterfaceColorCurve
  &NiagaraDataInterfaceVectorCurve
  &NiagaraDataInterfaceSpriteRendererInfo
  &NiagaraDataInterfaceArrayFloat3
  &NiagaraDataInterfaceSkeletalMesh
 ]
 usedDynamicInputs []
 usedModules [
  &self:SystemSpawnScript.SystemScriptSource.SystemScriptGraph.NiagaraNodeAssignment_0.SetVariables_#1
  &self:SystemSpawnScript.SystemScriptSource.SystemScriptGraph.NiagaraNodeAssignment_1.SetVariables_#2
  &SystemState
  &self:SystemSpawnScript.SystemScriptSource.SystemScriptGraph.NiagaraNodeAssignment_2.SetVariables_#3
 ]
}
emitters {
 NE_VC_Boost_Ribbon_Parent {
  emitterUpdateScript [
   &EmitterState(Life Cycle Mode=::NewEnumerator1, Loop Duration=1, UseLoopDelay=false)
   &VC_FilteredSocketCount()
   &VC_RateQualityHandler(Default Rate:NiagaraFloat=&MergeHelper_FloatSwitch(0=30, 1=30, 2=30, 3=30, Index=$User.GameModeID, Num=::NewEnumerator2))
   &VC_SpawnRibbon_v1(Number of Ribbons=$Emitter.PortCount, Spawn Time=-1, SpawnRate=$Emitter.Spawn Rate)
  ]
  particleSpawnScript [
   &VC_PaintedVariantsOverride(
    Black=<rgba 0.01 0.01 0.01 0.5>
    BurntSienna=<rgba 1 0.6172 0.2462 0.2>
    Cobalt=<rgba 0.1458 0.1579 0.8 0.2>
    Crimson=<rgba 4 0 0.1084 0.2>
    DefaultColor=<rgba 0.5 0.5 0.5 0.7>
    ForestGreen=<rgba 0 1 0.1022 0.2>
    Grey=<rgba 0.1 0.1 0.1 0.1>
    Lime=<rgba 0 1.5 0 0.302>
    Orange=<rgba 3 1.5 0 0.2>
    Pink=<rgba 2 0.2396 0.6797 0.2>
    Purple=<rgba 1 0 1 0.3>
    R1 Black:LinearColor=<rgba 0.002 0.002 0.002 0.2>
    R1 BurntSienna:LinearColor=<rgba 0.005 0.005 0.005 0.5>
    R1 Cobalt:LinearColor=<rgba 0.0152 0.0165 0.0833 0.15>
    R1 Crimson:LinearColor=<rgba 0.005 0.005 0.005 0.5>
    R1 DefaultColor=<rgba 0.01 0.01 0.01 1>
    R1 ForestGreen:LinearColor=<rgba 0.005 0.005 0.005 0.5>
    R1 Grey:LinearColor=<rgba 0.01 0.01 0.01 0.5>
    R1 Lime:LinearColor=<rgba 0.005 0.005 0.005 0.5>
    R1 Orange:LinearColor=<rgba 0.005 0.005 0.005 0.5>
    R1 Pink:LinearColor=<rgba 0.005 0.005 0.005 0.5>
    R1 Purple:LinearColor=<rgba 0.005 0.005 0.005 0.5>
    R1 Saffron:LinearColor=<rgba 0.005 0.005 0.005 0.5>
    R1 SkyBlue:LinearColor=<rgba 0.005 0.005 0.005 0.5>
    R1 TitaniumWhite:LinearColor=<rgba 0.005 0.005 0.005 0.5>
    R2 Black:LinearColor=<rgba 0.0191 0.0191 0.0191 0>
    R2 BurntSienna:LinearColor=<rgba 0.0642 0.0198 0.0079 0>
    R2 Cobalt:LinearColor=<rgba 0 0.0185 1 0>
    R2 Crimson:LinearColor=<rgba 0.4635 0 0.0126 0>
    R2 DefaultColor:LinearColor=<rgba 0.5 0.5 0.5 0>
    R2 ForestGreen:LinearColor=<rgba 0 0.5 0.0253 0>
    R2 Gold:LinearColor=<rgba 1 1 1 0>
    R2 Grey:LinearColor=<rgba 0.2031 0.2031 0.2031 0>
    R2 Lime:LinearColor=<rgba 1 1.5 0 0.302>
    R2 Orange:LinearColor=<rgba 0.6724 0.2346 0 0>
    R2 Pink:LinearColor=<rgba 0.5052 0.0605 0.3435 0>
    R2 Purple:LinearColor=<rgba 0.2789 0 0.6795 0>
    R2 Saffron:LinearColor=<rgba 0.9 0.69 0 0>
    R2 SkyBlue:LinearColor=<rgba 0 0.357 0.5365 0>
    R2 TitaniumWhite:LinearColor=<rgba 0.75 0.75 0.75 0>
    Saffron=<rgba 2 2 0 0.2>
    SkyBlue=<rgba 0 1 3 0.2>
    TitaniumWhite=<rgba 4 4 4 0.2>
    Use R1 Painted Overrides=true
    Use R2 Painted Overrides=true
   )
   &VC_InitializeRibbon_v1(
    Color=$StackContext.Color
    Custom Offset=true
    Enable Skeletal Position=true
    Lifetime=&MergeHelper_FloatSwitch(0=1, 1=1, 2=1, 3=1, Index=$User.GameModeID, Num=::NewEnumerator2)
    Local Vector Offset=&AddVectorToPosition(
     Position=&MergeHelper_VectorSwitch(0=<-45 0 0>, 1=<-45 0 0>, 2=<-45 0 0>, 3=<80 0 0>, Index=$User.GameModeID, Num=::NewEnumerator2)
     Vector=$User.DTCarOffset
    )
    Offset Amount=1
    Ribbon Width=15
   )
   &ApplyInitialForces() {enabled=false}
   &AddVelocity(Velocity=&MergeHelper_VectorSwitch(0=<-1500 0 0>, 1=<-1000 0 0>, 2=<-1000 0 0>, 3=<-2400 0 0>, Index=$User.GameModeID, Num=::NewEnumerator2)) {module=AddVelocity001}
   &Color(Color=&MakeCustomLinearColorFromBool(Bool=$User.bIsPainted, True Color=$User.PaintedColor, False Color=$StackContext.Color)) {enabled=false}
   /VehicleCosmetics/FX/VC_FXParents/Modules/DLR_ProductPreviews(GameModeID:NiagaraInt32=$System.GameModeID, bPreviewVehicleSpeed:NiagaraBool=$System.bPreviewVehicleSpeed)
  ]
  particleUpdateScript [
   &ParticleState()
   &InheritSourceMovement(
    Applied Position Scale=&VectorFromFloat(Value=&FloatFromCurve(FloatCurve=DataInterface() {propertyValues {Curve {keys [
          {interpMode=RCIM_Cubic, leaveTangent=-0.3746, tangentMode=RCTM_Break, time=0, value=1}
          {arriveTangent=-2.8875, interpMode=RCIM_Cubic, tangentMode=RCTM_Break, time=1, value=0}
         ]}}}))
    ApplyForce=false
    ApplyPosition=true
    ApplyPositionFalloff=false
   ) {enabled=false, module=InheritSourceMovement001}
   &InheritSourceMovement(
    Applied Position Scale=&VectorFromFloat(Value=&FloatFromCurve(FloatCurve=DataInterface() {propertyValues {Curve {
         defaultValue=1
         keys [
          {arriveTangent=-1.7885, interpMode=RCIM_Cubic, leaveTangent=-1.7885, tangentMode=RCTM_User, time=0.05, value=1}
          {interpMode=RCIM_Cubic, tangentMode=RCTM_User, time=1, value=0}
         ]
        }}}))
    ApplyForce=false
    ApplyPosition=true
    ApplyPositionFalloff=false
   )
   &ScaleColor(
    Scale Alpha=&FloatFromCurve(FloatCurve=DataInterface() {propertyValues {Curve {keys [<at 0 1>,<at 0.3 1>,<at 1 1>]}}}, Scale Curve=0.7)
    Scale RGB=&MakeCustomVectorFromBool(
     Bool=$User.bIsPainted
     True Vector=&MakeVectorFromLinearColorRGB(Color=$User.PaintedColor)
     False Vector=&VectorFromCurve(VectorCurve=DataInterface() {propertyValues {
        XCurve {keys [<at 0 1>,<at 0.2 1>,<at 0.4 0>,<at 0.8 0>,<at 1 0>]}
        YCurve {keys [<at 0 0>,<at 0.2 0>,<at 0.4 1>,<at 0.8 0>,<at 1 0>]}
        ZCurve {keys [<at 0 0>,<at 0.2 0>,<at 0.4 0>,<at 0.8 1>,<at 1 0>]}
       }})
    )
    ScaleA=true
    ScaleRGB=true
   )
   &DynamicMaterialParameters(
    Index 0 Param 1=30
    Index 0 Param 2=16
    Index 0 Param 3=5
    Index 0 Param 4=0.8
    Module.Index0Param0Write=true
    Module.Index0Param1Write=true
    Module.Index0Param2Write=true
    Module.Index0Param3Write=true
   )
   &SolveForcesAndVelocity(Clamp Velocity=false, Limit Acceleration=false)
   &ScaleRibbonWidth(Ribbon Width Scale=2.8)
   &VC_UpdateRibbon_v1(Enable Skeletal Position=true)
  ]
  rendererClasses [&NiagaraRibbonRendererProperties]
  renderers [
   {rendererClass=&NiagaraRibbonRendererProperties}
   {rendererClass=&NiagaraRibbonRendererProperties, rendererIndex=1}
   {rendererClass=&NiagaraRibbonRendererProperties, rendererIndex=2}
   {rendererClass=&NiagaraRibbonRendererProperties, rendererIndex=3}
  ]
 }
 NE_VC_Boost_Sprite_Parent {
  emitterUpdateScript [
   &EmitterState(Inactive Response=::NewEnumerator1, Life Cycle Mode=::NewEnumerator1, Loop Duration=0.2, UseLoopDelay=false)
   &VC_FilteredSocketCount()
   &SpawnRate(SpawnRate=50, Use Spawn Probability=false) {enabled=false}
   &SpawnBurst_Instantaneous(Spawn Count=12, Use Spawn Probability=false, UseLoopCountLimit=false)
  ]
  particleSpawnScript [
   &VC_PaintedVariantsOverride(
    Black=<rgba 0.0191 0.0191 0.0191 0.03>
    BurntSienna=<rgba 1 0.3088 0.1233 0.03>
    Cobalt=<rgba 0.0002 0.1075 2 0.03>
    Crimson=<rgba 0.4635 0 0.0126 0.03>
    DefaultColor=&RandomRangeLinearColor(Minimum=<rgba 0.8 0.016 0.016 0.03>, Maximum=<rgba 0.9 0.045 0.045 0.03>)
    ForestGreen=<rgba 0 1 0.0505 0.03>
    Grey=<rgba 0.0191 0.0191 0.0191 0.03>
    Lime=<rgba 0.5 1 0 0.03>
    Orange=<rgba 1 0.3488 0 0.03>
    Pink=<rgba 1 0.1198 0.6799 0.03>
    Purple=<rgba 0.4104 0 1 0.03>
    R1 Black:LinearColor=<rgba 0.0191 0.0191 0.0191 0.03>
    R1 BurntSienna:LinearColor=<rgba 0.0642 0.0198 0.0079 0.03>
    R1 Cobalt:LinearColor=<rgba 0 0.037 2 0.03>
    R1 Crimson:LinearColor=<rgba 0.4635 0 0.0126 0.03>
    R1 DefaultColor=<rgba 0.7 0.474 0.4746 0.03>
    R1 ForestGreen:LinearColor=<rgba 0 1 0.0505 0.03>
    R1 Gold:LinearColor=<rgba 1 1 1 0.03>
    R1 Grey:LinearColor=<rgba 0.0191 0.0191 0.0191 0.03>
    R1 Lime:LinearColor=<rgba 0.2943 0.5885 0 0.03>
    R1 Orange:LinearColor=<rgba 1 0.3488 0 0.03>
    R1 Pink:LinearColor=<rgba 1 0.1198 0.6799 0.03>
    R1 Purple:LinearColor=<rgba 0.4104 0 1 0.03>
    R1 Saffron:LinearColor=<rgba 0.9 0.69 0 0.03>
    R1 SkyBlue:LinearColor=<rgba 0 0.357 0.5365 0.03>
    R1 TitaniumWhite:LinearColor=<rgba 0.75 0.75 0.75 0.03>
    R2 Black:LinearColor=<rgba 0.0191 0.0191 0.0191 0.03>
    R2 BurntSienna:LinearColor=<rgba 0.0642 0.0198 0.0079 0.03>
    R2 Cobalt:LinearColor=<rgba 0 0.037 2 0.03>
    R2 Crimson:LinearColor=<rgba 0.4635 0 0.0126 0.03>
    R2 DefaultColor:LinearColor=<rgba 0.8177 0.6125 0.2896 0.03>
    R2 ForestGreen:LinearColor=<rgba 0 1 0.0505 0.03>
    R2 Grey:LinearColor=<rgba 0.2031 0.2031 0.2031 0.03>
    R2 Lime:LinearColor=<rgba 0.2943 0.5885 0 0.03>
    R2 Orange:LinearColor=<rgba 1 0.3488 0 0.03>
    R2 Pink:LinearColor=<rgba 1 0.1198 0.6799 0.03>
    R2 Purple:LinearColor=<rgba 0.2789 0 0.6795 0.03>
    R2 Saffron:LinearColor=<rgba 0.9 0.69 0 0.03>
    R2 SkyBlue:LinearColor=<rgba 0 0.357 0.5365 0.03>
    R2 TitaniumWhite:LinearColor=<rgba 0.75 0.75 0.75 0.03>
    Saffron=<rgba 0.9 0.69 0 0.03>
    SkyBlue=<rgba 0 0.357 0.5365 0.03>
    TitaniumWhite=<rgba 0.75 0.75 0.75 0.03>
    Use R1 Painted Overrides=true
    Use R2 Painted Overrides=true
   )
   &InitializeParticle(
    Color=$Particles.Color
    Lifetime=0.4
    Mesh Scale=<1 2 1>
    Mesh Scale Mode=::NewEnumerator4
    UsePositionOffset=false
    Write Material Random=true
   )
   &PartitionParticles(Number Of Partitions=$Emitter.PortCount)
   &SkeletalMeshLocation(
    Bone / Socket Index=$Particles.PartitionParticles.Partition
    Bone / Socket Sampling Mode=::NewEnumerator5
    MaskSpawnGroup=false
    Mesh Sampling Type=::NewEnumerator6
    Sampled Position Offset=&AddVectorToPosition(Position=<-45 0 0>, Vector=$User.DTCarOffset)
    Skeletal Mesh=$System.SkeletalMesh
   )
   &AddVelocity(Velocity=<-1500 0 0>) {enabled=false}
   &InitialMeshOrientation(Mesh Orientation Mode=::NewEnumerator2, Orientation Coordinate Space=::Local, Orientation Vector=<-1 0 0>, Use Rotation Vector=false)
  ]
  particleUpdateScript [
   &ParticleState()
   &SolveForcesAndVelocity(Clamp Velocity=false, Limit Acceleration=false) {enabled=false}
   &DynamicMaterialParameters(
    Index 1 Param 1=&FloatFromCurve(
     FloatCurve=DataInterface() {propertyValues {Curve {keys [
         {arriveTangent=-13.683, interpMode=RCIM_Cubic, leaveTangent=-13.683, tangentMode=RCTM_User, time=0.05, value=2}
         <at 0.25 -4 RCIM_Cubic>
         <at 0.9872 2 RCIM_Cubic>
        ]}}}
     Scale Curve=&Waveform("Waveform [1]"=::NewEnumerator0, "[1] Frequency"=3, "[1] Amplitude Min/Max"=<0.1 0.1>)
    )
    Index 1 Param 2=1
    Module.Index0Float0Write=true
    Module.Index0Float1Write=true
    Module.Index0Float2Write=false
    Module.Index0Float3Write=false
    Module.Index1Float0Write=true
    Module.Index1Float1Write=true
    Module.Index1Float2Write=true
    Module.Index1Float3Write=true
    Write Parameter Index 1=::NewEnumerator1
   )
  ]
  rendererClasses [&NiagaraMeshRendererProperties, &NiagaraSpriteRendererProperties]
  renderers [
   {rendererClass=&NiagaraMeshRendererProperties, rendererIndex=0}
   {rendererClass=&NiagaraMeshRendererProperties, rendererIndex=1}
   {rendererClass=&NiagaraMeshRendererProperties, rendererIndex=2}
   {rendererClass=&NiagaraSpriteRendererProperties, rendererIndex=3}
  ]
 }
 ParticlesOptimized {
  emitterUpdateScript [
   &EmitterState()
   &VC_FilteredSocketCount()
   &SpawnRate(
    SpawnRate=&MakeCustomFloatFromBool(Bool=$User.bIsFrontendPreview, True Float=30, False Float=40)
    Use Spawn Probability=false
   )
   &SpawnRate(
    SpawnGroup=1
    SpawnRate=&MakeCustomFloatFromBool(Bool=$User.bIsFrontendPreview, True Float=8, False Float=30)
    Use Spawn Probability=false
   ) {module=SpawnRate002}
   &SpawnRate(
    SpawnGroup=2
    SpawnRate=&MakeCustomFloatFromBool(Bool=$User.bIsFrontendPreview, True Float=13, False Float=30)
    Use Spawn Probability=false
   ) {module=SpawnRate003}
  ]
  particleSpawnScript [
   &VC_PaintedVariantsOverride(
    Black=<rgba 0.0191 0.0191 0.0191 0.7>
    BurntSienna=<rgba 0.4271 0.1044 0.0178 0.7>
    Cobalt=<rgba 0.0833 0 1 0.7>
    Crimson=<rgba 2 0 0.0542 0.7>
    DefaultColor=<rgba 1 1 1 0.7>
    ForestGreen=<rgba 0 0.5 0.025 0.7>
    Gold:LinearColor=<rgba 1 1 1 0.7>
    Grey=<rgba 0.7 0.7 0.7 0.7>
    Lime=<rgba 0.1963 0.589 0 0.7>
    Orange=<rgba 1 0.8333 0 0.7>
    Pink=<rgba 0.5052 0.0605 0.0605 0.7>
    Purple=<rgba 0.7 0 0.5833 0.7>
    Saffron=<rgba 0.7 0.7 0 0.5>
    SkyBlue=<rgba 0 0.1167 0.7 0.7>
    TitaniumWhite=<rgba 0.75 0.75 0.75 0.7>
   )
   /SetVariables(Particles.NewTagFlags:NiagaraInt32=$Emitter.SpawnGroup) {isSetParameters=true}
   &InitializeParticle(
    Color=$StackContext.Color
    Lifetime=/VehicleCosmetics/FX/Niagara/Modules/MergeEmitters/VC_FloatBySpawnGroup(
     SpawnGroup or ExecIndex:NiagaraInt32=0
     CustomInt:NiagaraInt32=0
     MaximumSpawnGroups:NiagaraInt32=3
     Float SG 0:NiagaraFloat=&RandomRangeFloat(Minimum=0.35, Maximum=0.6)
     Float SG 1:NiagaraFloat=&MakeCustomFloatFromBool(Bool=$User.bIsFrontendPreview, True Float=&RandomRangeFloat(Minimum=0.8, Maximum=1), False Float=1)
     Float SG 2:NiagaraFloat=&MakeCustomFloatFromBool(Bool=$User.bIsFrontendPreview, True Float=&RandomRangeFloat(Minimum=0.8, Maximum=1.2), False Float=1)
     Float SG 3:NiagaraFloat=0
     Float SG 4:NiagaraFloat=0
     Float SG 5:NiagaraFloat=0
    )
    Position Offset=<0 300 0>
    Sprite Size Mode=::NewEnumerator3
    Uniform Sprite Size=&VC_FloatBySpawnGroup(
     SpawnGroup or ExecIndex:NiagaraInt32=0
     CustomInt:NiagaraInt32=0
     MaximumSpawnGroups:NiagaraInt32=3
     Float SG 0:NiagaraFloat=&RandomRangeFloat(Minimum=4, Maximum=10)
     Float SG 1:NiagaraFloat=&RandomRangeFloat(Minimum=4.5, Maximum=10)
     Float SG 2:NiagaraFloat=&RandomRangeFloat(Minimum=9, Maximum=15)
     Float SG 3:NiagaraFloat=0
     Float SG 4:NiagaraFloat=0
     Float SG 5:NiagaraFloat=0
    )
    UsePositionOffset=false
    Write Material Random=true
   )
   &PartitionParticles(Number Of Partitions=$Emitter.PortCount)
   &SkeletalMeshLocation(
    Bone / Socket Index=$Particles.PartitionParticles.Partition
    Bone / Socket Sampling Mode=::NewEnumerator5
    MaskSpawnGroup=false
    Mesh Sampling Type=::NewEnumerator6
    Sampled Position Offset=&AddVectorToPosition(
     Position=/VehicleCosmetics/FX/Niagara/Modules/MergeEmitters/VC_VectorBySpawnGroup(
      SpawnGroup or ExecIndex:NiagaraInt32=0
      CustomInt:NiagaraInt32=0
      MaximumSpawnGroups:NiagaraInt32=3
      Vector Group 0:Vector3f=<-70 0 0>
      Vector Group 1:Vector3f=&MakeCustomVectorFromBool(Bool=$User.bIsFrontendPreview, True Vector=<-70 0 1>, False Vector=<0 0 0>)
      Vector Group 2:Vector3f=&MakeCustomVectorFromBool(Bool=$User.bIsFrontendPreview, True Vector=<-70 0 0>, False Vector=<0 0 0>)
      Vector Group 3:Vector3f=<0 0 0>
      Vector Group 4:Vector3f=<0 0 0>
      Vector Group 5:Vector3f=<0 0 0>
     )
     Vector=$User.DTCarOffset
    )
    Skeletal Mesh=$System.SkeletalMesh
   )
   &AddVelocity(Velocity=&VC_VectorBySpawnGroup(
     SpawnGroup or ExecIndex:NiagaraInt32=0
     CustomInt:NiagaraInt32=$User.GameModeID
     MaximumSpawnGroups:NiagaraInt32=4
     Vector Group 0:Vector3f=&MakeCustomVectorFromBool(
      Bool=$User.bIsFrontendPreview
      True Vector=&RandomRangeVector(Minimum=<-240 -30 0>, Maximum=<-480 30 0>)
      False Vector=&RandomRangeVector(Minimum=<-1000 0 0>, Maximum=<-1000 0 0>)
     )
     Vector Group 1:Vector3f=&MakeCustomVectorFromBool(Bool=$User.bIsFrontendPreview, True Vector=<-1000 -10 0>, False Vector=<-1000 10 0>)
     Vector Group 2:Vector3f=&MakeCustomVectorFromBool(Bool=$User.bIsFrontendPreview, True Vector=<-1000 -10 0>, False Vector=<-1000 10 0>)
     Vector Group 3:Vector3f=&RandomRangeVector(Minimum=<-1000 -30 0>, Maximum=<-1000 30 1>)
     Vector Group 4:Vector3f=<0 0 0>
     Vector Group 5:Vector3f=<0 0 0>
    ))
   &Color(Color=&MakeCustomLinearColorFromBool(Bool=$User.bIsFrontendPreview, True Color=<rgba 1 1 1 0.7>, False Color=<rgba 1 1 1 1>)) {enabled=false}
   &ShapeLocation(
    MaskSpawnGroup=false
    Override Seed=false
    Sphere Radius=&VC_FloatBySpawnGroup(
     SpawnGroup or ExecIndex:NiagaraInt32=0
     CustomInt:NiagaraInt32=0
     MaximumSpawnGroups:NiagaraInt32=3
     Float SG 0:NiagaraFloat=20
     Float SG 1:NiagaraFloat=12
     Float SG 2:NiagaraFloat=&MakeCustomFloatFromBool(Bool=$User.bIsFrontendPreview, True Float=15, False Float=12)
     Float SG 3:NiagaraFloat=0
     Float SG 4:NiagaraFloat=0
     Float SG 5:NiagaraFloat=0
    )
    UseOwnerScale=true
   )
   &DLR_ProductPreviews(GameModeID:NiagaraInt32=$System.GameModeID, bPreviewVehicleSpeed:NiagaraBool=$System.bPreviewVehicleSpeed)
  ]
  particleUpdateScript [
   &ParticleState()
   &CameraOffset(Camera Offset Amount=20) {enabled=false}
   &ScaleSpriteSize(
    Scale Sprite Size Mode=::NewEnumerator0
    Uniform Scale Factor=&VC_FloatBySpawnGroup(
     SpawnGroup or ExecIndex:NiagaraInt32=0
     CustomInt:NiagaraInt32=0
     MaximumSpawnGroups:NiagaraInt32=3
     Float SG 0:NiagaraFloat=&FloatFromCurve(
      FloatCurve=DataInterface() {propertyValues {Curve {keys [
          {interpMode=RCIM_Cubic, tangentMode=RCTM_Break, time=0, value=1.5}
          {arriveTangent=-0.9784, interpMode=RCIM_Cubic, tangentMode=RCTM_Break, time=0.7, value=1}
          {interpMode=RCIM_Cubic, tangentMode=RCTM_Break, time=1, value=0}
         ]}}}
      Scale Curve=2.8
     )
     Float SG 1:NiagaraFloat=&FloatFromCurve(
      FloatCurve=DataInterface() {propertyValues {Curve {keys [
          {interpMode=RCIM_Cubic, tangentMode=RCTM_Break, time=0, value=1.5}
          {arriveTangent=-0.9784, interpMode=RCIM_Cubic, tangentMode=RCTM_Break, time=0.7, value=1}
          {interpMode=RCIM_Cubic, tangentMode=RCTM_Break, time=1, value=0}
         ]}}}
      Scale Curve=5
     )
     Float SG 2:NiagaraFloat=&FloatFromCurve(
      FloatCurve=DataInterface() {propertyValues {Curve {keys [
          {interpMode=RCIM_Cubic, tangentMode=RCTM_Break, time=0, value=1.5}
          {arriveTangent=-0.9784, interpMode=RCIM_Cubic, tangentMode=RCTM_Break, time=0.7, value=1}
          {interpMode=RCIM_Cubic, tangentMode=RCTM_Break, time=1, value=0}
         ]}}}
      Scale Curve=2.8
     )
     Float SG 3:NiagaraFloat=0
     Float SG 4:NiagaraFloat=0
     Float SG 5:NiagaraFloat=0
    )
   )
   &InheritSourceMovement(
    Applied Position Scale=&VectorFromFloat(Value=&FloatFromCurve(FloatCurve=DataInterface() {propertyValues {Curve {keys [
          {interpMode=RCIM_Cubic, leaveTangent=-0.0191, tangentMode=RCTM_Break, time=0.3, value=1}
          {arriveTangent=-5.0309, interpMode=RCIM_Cubic, tangentMode=RCTM_Break, time=1, value=0}
         ]}}}))
    ApplyForce=false
    ApplyPosition=true
    ApplyPositionFalloff=false
   ) {enabled=false}
   &Drag(
    Drag=&VC_FloatBySpawnGroup(
     SpawnGroup or ExecIndex:NiagaraInt32=0
     CustomInt:NiagaraInt32=0
     MaximumSpawnGroups:NiagaraInt32=3
     Float SG 0:NiagaraFloat=&MakeCustomFloatFromBool(
      Bool=$User.bIsFrontendPreview
      True Float=&MakeFloatFromVector(
       Vector=&VectorFromFloat(Value=&FloatFromCurve(FloatCurve=DataInterface() {propertyValues {Curve {keys [<at 0 1 RCIM_Cubic>,<at 1 0 RCIM_Cubic>]}}}))
       Channel=::NewEnumerator0
      )
      False Float=&MakeFloatFromVector(
       Vector=&VectorFromFloat(Value=&FloatFromCurve(FloatCurve=DataInterface() {propertyValues {Curve {keys [<at 0 1 RCIM_Cubic>,<at 1 0 RCIM_Cubic>]}}}))
       Channel=::NewEnumerator0
      )
     )
     Float SG 1:NiagaraFloat=&MakeCustomFloatFromBool(Bool=$User.bIsFrontendPreview, True Float=2, False Float=0)
     Float SG 2:NiagaraFloat=&MakeCustomFloatFromBool(Bool=$User.bIsFrontendPreview, True Float=1, False Float=0)
     Float SG 3:NiagaraFloat=0
     Float SG 4:NiagaraFloat=0
     Float SG 5:NiagaraFloat=0
    )
    Use Linear Drag=true
    Use Rotational Drag=true
   )
   &ScaleColor(
    Scale Alpha=&FloatFromCurve(
     FloatCurve=DataInterface() {propertyValues {Curve {keys [<at 0 1>,<at 0.1 1>,<at 1 1>]}}}
     Scale Curve=&VC_FloatBySpawnGroup(
      SpawnGroup or ExecIndex:NiagaraInt32=0
      CustomInt:NiagaraInt32=0
      MaximumSpawnGroups:NiagaraInt32=3
      Float SG 0:NiagaraFloat=2
      Float SG 1:NiagaraFloat=1
      Float SG 2:NiagaraFloat=1
      Float SG 3:NiagaraFloat=0
      Float SG 4:NiagaraFloat=0
      Float SG 5:NiagaraFloat=0
     )
    )
    Scale RGB=&MakeCustomVectorFromBool(
     Bool=$User.bIsPainted
     True Vector=&Multiply_VectorByFloat(Vector=&MakeVectorFromLinearColorRGB(Color=$User.PaintedColor2), Float=$User.Intensity_color)
     False Vector=&VectorFromCurve(
      VectorCurve=DataInterface() {propertyValues {
        XCurve {keys [<at 0 1>,<at 0.2 1>,<at 0.4 0>,<at 0.8 0>,<at 1 0>]}
        YCurve {keys [<at 0 0>,<at 0.2 0>,<at 0.4 1>,<at 0.8 0>,<at 1 0>]}
        ZCurve {keys [<at 0 0>,<at 0 0>,<at 0.4 0>,<at 0.8 1>,<at 1 0>]}
       }}
      Scale Curve=&VC_VectorBySpawnGroup(
       SpawnGroup or ExecIndex:NiagaraInt32=0
       CustomInt:NiagaraInt32=0
       MaximumSpawnGroups:NiagaraInt32=3
       Vector Group 0:Vector3f=<1.5 1.5 1.5>
       Vector Group 1:Vector3f=<1 1 1>
       Vector Group 2:Vector3f=<1 1 1>
       Vector Group 3:Vector3f=<0 0 0>
       Vector Group 4:Vector3f=<0 0 0>
       Vector Group 5:Vector3f=<0 0 0>
      )
     )
    )
    ScaleA=true
    ScaleRGB=true
   )
   &SubUVAnimation(
    End Frame Range Override=8
    Output Alternate Renderer Binding=true
    "Sprite Renderer "=DataInterface() {propertyValues {SpriteRenderer=&self:ParticlesOptimized_0.NiagaraSpriteRendererProperties_3}}
    SubUV Animation Mode:ENiagara_SubUVLookupModeV2=::NewEnumerator1
    SubUV Curve=DataInterface() {propertyValues {Curve {keys [{arriveTangent=4.1785, time=0, value=0}, {arriveTangent=0.2086, leaveTangent=0.2086, time=1, value=1}]}}}
    UseEndFrame=true
    UseStartFrame=true
   )
   &SubUVAnimation(
    End Frame Range Override=8
    Output Alternate Renderer Binding=true
    "Sprite Renderer "=DataInterface() {propertyValues {SpriteRenderer=&self:ParticlesOptimized_0.NiagaraSpriteRendererProperties_2}}
    SubUV Animation Mode:ENiagara_SubUVLookupModeV2=::NewEnumerator1
    SubUV Curve=DataInterface() {propertyValues {Curve {keys [{arriveTangent=4.1785, time=0, value=0}, {arriveTangent=0.2086, leaveTangent=0.2086, time=1, value=1}]}}}
    UseEndFrame=true
    UseStartFrame=true
   ) {module=SubUVAnimation001}
   &SubUVAnimation(
    End Frame Range Override=8
    Output Alternate Renderer Binding=true
    "Sprite Renderer "=DataInterface() {propertyValues {SpriteRenderer=&self:ParticlesOptimized_0.NiagaraSpriteRendererProperties_4}}
    SubUV Animation Mode:ENiagara_SubUVLookupModeV2=::NewEnumerator1
    SubUV Curve=DataInterface() {propertyValues {Curve {keys [{arriveTangent=4.1785, time=0, value=0}, {arriveTangent=0.2086, leaveTangent=0.2086, time=1, value=1}]}}}
    UseEndFrame=true
    UseStartFrame=true
   ) {module=SubUVAnimation002}
   &InheritSourceMovement(
    Applied Position Scale=&Multiply_VectorByFloat(
     Vector=<1 1 1>
     Float=&FloatFromCurve(FloatCurve=DataInterface() {propertyValues {Curve {
         defaultValue=1
         keys [
          {arriveTangent=-1.7885, interpMode=RCIM_Cubic, leaveTangent=-1.7885, tangentMode=RCTM_User, time=0.05, value=1}
          {interpMode=RCIM_Cubic, tangentMode=RCTM_User, time=1, value=0}
         ]
        }}})
    )
    ApplyForce=false
    ApplyPosition=true
    ApplyPositionFalloff=false
   ) {module=InheritSourceMovement001}
   &DynamicMaterialParameters(
    Index 0 Param 3=1
    Index 0 Param 4=1
    Module.Index0Param0Write=false
    Module.Index0Param1Write=false
    Module.Index0Param2Write=true
    Module.Index0Param3Write=true
   ) {enabled=false}
   /SetVariables(Particles.DynamicIndex_1:Vector4f=&MakeVector4(
     X=&RandomRangeFloat(Minimum=0, Maximum=1)
     Y=&MakeCustomFloatFromBool(
      Bool=$User.bIsPainted
      True Float=3
      False Float=&MakeCustomFloatFromBool(Bool=$User.bIsFrontendPreview, True Float=5, False Float=2)
     )
     Z=&MakeCustomFloatFromBool(Bool=$User.bIsPainted, True Float=$User.Distortion, False Float=0.02)
     W=0.5
    )) {enabled=false, isSetParameters=true}
   /SetVariables(Particles.DynamicIndex_2:Vector4f=&MakeVector4(
     X=&RandomRangeFloat(Minimum=0, Maximum=1)
     Y=&MakeCustomFloatFromBool(
      Bool=$User.bIsPainted
      True Float=3
      False Float=&MakeCustomFloatFromBool(Bool=$User.bIsFrontendPreview, True Float=5, False Float=2)
     )
     Z=&MakeCustomFloatFromBool(Bool=$User.bIsPainted, True Float=$User.Distortion, False Float=0.02)
     W=0.5
    )) {enabled=false, isSetParameters=true}
   &SolveForcesAndVelocity(Clamp Velocity=false, Limit Acceleration=false)
  ]
  rendererClasses [&NiagaraSpriteRendererProperties]
  renderers [
   {rendererClass=&NiagaraSpriteRendererProperties}
   {rendererClass=&NiagaraSpriteRendererProperties, rendererIndex=1}
   {rendererClass=&NiagaraSpriteRendererProperties, rendererIndex=2}
  ]
 }
 mesh_glow {
  emitterUpdateScript [
   &EmitterState(
    Inactive Response=::NewEnumerator1
    Life Cycle Mode=::NewEnumerator1
    Loop Behavior=::NewEnumerator1
    Loop Duration Mode=::NewEnumerator1
    UseLoopDelay=false
   )
   &VC_FilteredSocketCount()
   &SpawnBurst_Instantaneous(Spawn Count=2, Use Spawn Probability=false, UseLoopCountLimit=false)
   &SpawnBurst_Instantaneous(Spawn Count=2, Spawn Group=1, Use Spawn Probability=false, UseLoopCountLimit=false) {module=SpawnBurst_Instantaneous001}
   &SpawnRate(SpawnRate=2.2, Use Spawn Probability=false) {enabled=false}
  ]
  particleSpawnScript [
   &VC_PaintedVariantsOverride(
    Black=<rgba 1 1 1 0.05>
    BurntSienna=<rgba 0.526 0.1624 0.0648 0.15>
    Cobalt=<rgba 0 0.0833 2 0.15>
    Crimson=<rgba 1 0.0677 0.093 0.06>
    DefaultColor=<rgba 1 0.1833 0.02 0.15>
    ForestGreen=<rgba 0 1 0.0505 0.15>
    Grey=<rgba 1 1 1 0.05>
    Lime=<rgba 0.25 1 0 0.05>
    Orange=<rgba 0.75 0.1 0 0.15>
    Pink=<rgba 1 0.3 0.825 0.15>
    Purple=<rgba 0.5 0 2 0.15>
    R1 Black:LinearColor=<rgba 14 14 14 1>
    R1 BurntSienna:LinearColor=<rgba 14 4.3225 1.7257 2>
    R1 Cobalt:LinearColor=<rgba 0.2 0.65 2 1.5>
    R1 Crimson:LinearColor=<rgba 4 0.2292 0.2292 1>
    R1 DefaultColor=<rgba 14 0.28 0.28 1>
    R1 ForestGreen:LinearColor=<rgba 0 4 0.2021 1>
    R1 Grey:LinearColor=<rgba 14 14 14 1>
    R1 Lime:LinearColor=<rgba 2.5547 3 0.3281 1.5>
    R1 Orange:LinearColor=<rgba 4 0.9097 0.2917 1>
    R1 Pink:LinearColor=<rgba 3 0.0625 1.9318 1>
    R1 Purple:LinearColor=<rgba 0.75 0 5 2>
    R1 Saffron:LinearColor=<rgba 5 4.2014 0.2083 1.5>
    R1 SkyBlue:LinearColor=<rgba 0.5 2 5 1>
    R1 TitaniumWhite:LinearColor=<rgba 1 1 1 1>
    Saffron=<rgba 1 0.8236 0.0016 0.1>
    SkyBlue=<rgba 0.3958 0.8771 2 0.05>
    TitaniumWhite=<rgba 0.3 0.3 0.3 0.15>
    Use R1 Painted Overrides=true
   )
   /SetVariables(Particles.VisibilityTag=$Emitter.SpawnGroup) {isSetParameters=true}
   &InitializeParticle(
    Color=$Particles.Color
    Mesh Scale=<1 1 0.65>
    Mesh Scale Mode=::NewEnumerator4
    Sprite Size=<160 80>
    Sprite Size Mode=::NewEnumerator4
    UsePositionOffset=false
    Write Material Random=true
   )
   &PartitionParticles(Number Of Partitions=$Emitter.PortCount)
   &SkeletalMeshLocation(
    Bone / Socket Index=$Particles.PartitionParticles.Partition
    Bone / Socket Sampling Mode=::NewEnumerator5
    MaskSpawnGroup=false
    Mesh Sampling Type=::NewEnumerator6
    Sampled Position Offset=&MergeHelper_VectorSwitch(
     0=&AddVectorToPosition(Position=<-45 0 0>, Vector=$User.DTCarOffset)
     1=&AddVectorToPosition(Position=<-65 0 0>, Vector=$User.DTCarOffset)
     2=?
     3=?
     Index=&SpawnGroup()
     Num=::NewEnumerator0
    )
    Skeletal Mesh=$System.SkeletalMesh
   )
   &AddVelocity(Velocity=<0 0 0>)
   &InitialMeshOrientation(
    Mesh Orientation Mode=::NewEnumerator2
    Orientation Vector=<-1 0 0>
    Rotation=&SelectVectorFromArray(
     Direct Array Index=$Particles.PartitionParticles.Partition
     Vector Selection Array=DataInterface() {propertyValues {FloatData [<0 0 0>, <0 0 0>]}}
    )
    Rotation Coordinate Space=::NewEnumerator2
    Use Rotation Vector=false
   )
  ]
  particleUpdateScript [
   &ParticleState(Kill Particles When Lifetime Has Elapsed=false, Loop Particles Lifetime=true)
   &SolveForcesAndVelocity(Clamp Velocity=false, Limit Acceleration=false)
   &DynamicMaterialParameters(
    Module.Index0Float0Write=true
    Module.Index0Float1Write=true
    Module.Index0Float2Write=false
    Module.Index0Float3Write=false
    Module.Index1Float0Write=true
    Module.Index1Float1Write=true
    Module.Index1Float2Write=true
    Module.Index1Float3Write=true
    Write Parameter Index 1=::NewEnumerator1
   )
  ]
  rendererClasses [&NiagaraSpriteRendererProperties, &NiagaraMeshRendererProperties]
  renderers [
   {rendererClass=&NiagaraSpriteRendererProperties}
   {rendererClass=&NiagaraMeshRendererProperties, rendererIndex=1}
   {rendererClass=&NiagaraSpriteRendererProperties, rendererIndex=2}
  ]
 }
}
summary {
 systemName=NS_Boost_Glitch
 emitters [
  {emitterName=NE_VC_Boost_Ribbon_Parent, rendererClasses [&NiagaraRibbonRendererProperties]}
  {emitterName=ParticlesOptimized, rendererClasses [&NiagaraSpriteRendererProperties]}
  {
   emitterName=NE_VC_Boost_Sprite_Parent
   rendererClasses [&NiagaraMeshRendererProperties, &NiagaraSpriteRendererProperties]
  }
  {emitterName=mesh_glow, rendererClasses [&NiagaraSpriteRendererProperties, &NiagaraMeshRendererProperties]}
 ]
 userVariables [
  User.BackTrail_power {type=NiagaraFloat, defaultValue=0.5, description=""}
  User.bIsFrontendPreview {type=NiagaraBool, defaultValue=true}
  User.bIsPainted {type=NiagaraBool}
  User.blsIsCoroColor {type=NiagaraBool}
  User.bPreviewVehicleSpeed {type=NiagaraBool}
  User.ColorVariantID {type=NiagaraInt32}
  User.Distortion {type=NiagaraFloat, defaultValue=1, description=""}
  User.DTCarOffset {type=Vector3f, description="", defaultValue {struct=Vector3f, value=<0 0 0>}}
  User.GameModeID {type=NiagaraInt32, defaultValue=3}
  User.Intensity_color {type=NiagaraFloat, defaultValue=1, description=""}
  User.PaintedColor {type=LinearColor, description="", defaultValue {struct=LinearColor, value=<rgba 1 1 1 1>}}
  User.PaintedColor2 {type=LinearColor, description="", defaultValue {struct=LinearColor, value=<rgba 1 1 1 1>}}
  User.PaintedColor_Particles {type=LinearColor, description="", defaultValue {struct=LinearColor, value=<rgba 1 1 1 1>}}
  User.PaintedScalar {type=NiagaraFloat, defaultValue=1, description=""}
  User.PortCount {type=NiagaraInt32, defaultValue=2}
 ]
}
