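# Translucent unlit Niagara sprite material for the glitchy "bits" particles of the Glitch vehicle boost cosmetic.
# Layered panning pixel-noise samples, pulsed by sine time, warp the UVs of glitch textures whose color channels are shuffled for corrupted color.
# A second distortion chain scrolls a tiled square mask to form the blocky shapes and opacity.
# Dynamic parameters control highlights, brightness, distortion and inner detail, while particle color tints emissive and scales alpha.
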Material /VehicleCosmetics/Boosts/Boost_Glitch/Material/M_Boost_glitch_Bits
graph {
 properties {bUsedWithNiagaraSprites=true, blendMode=BLEND_Translucent, pixelDepthOffsetMode=PDOM_Legacy, shadingModel=MSM_Unlit, usedShadingModels=Unlit}
 nodes [
  n1 = DynamicParameter() {paramNames [Highlights,Brightness,Distortion,InnerDetail]}
  n2 = TextureCoordinate()
  n3 = TextureSample(UVs=@n2) {texture=/VehicleCosmetics/Boosts/Boost_Glitch/Texture/T_VCVFX_Boost_Glitch_Glitch_FB_03_RGB}
  n4 = Desaturation(None=@n3.RGB) {luminanceFactors=<rgba 0.3 0.59 0.11 0>}
  n5 = TextureCoordinate() {uTiling=0.5, vTiling=0.1}
  n6 = Multiply(A=ScalarParameter() {defaultValue=0.01, parameterName=GlitchSpeed}, B=Sine(None=Time()))
  n7 = TextureCoordinate() {uTiling=-0.2, vTiling=2.5}
  n8 = AppendVector(
   A=TextureSample(UVs=Panner(Coordinate=@n5, Time=@n6) {speedX=0.1, speedY=-0.5}) {texture=/VehicleCosmetics/Boosts/Boost_Glitch/Texture/T_VCVFX_Boost_Glitch_Pixel_Alpha}.R
   B=TextureSample(UVs=Panner(Coordinate=@n7, Time=@n6) {speedX=-0.1, speedY=0.1}) {texture=&T_VCVFX_Boost_Glitch_Pixel_Alpha}.G
  )
  n9 = TextureCoordinate() {uTiling=3, vTiling=0.25}
  n10 = TextureCoordinate() {vTiling=-0.1}
  n11 = AppendVector(
   A=TextureSample(UVs=Panner(Coordinate=@n9, Time=@n6) {speedX=-0.5, speedY=0.1}) {texture=&T_VCVFX_Boost_Glitch_Pixel_Alpha}.R
   B=TextureSample(UVs=Panner(Coordinate=@n10, Time=@n6) {speedX=-0.1, speedY=0.1}) {texture=&T_VCVFX_Boost_Glitch_Pixel_Alpha}.G
  )
  n12 = Time()
  n13 = LinearInterpolate(
   A=@n2
   B=Add(A=@n2, B=Add(A=@n8, B=@n11))
   Alpha=Multiply(A=ScalarParameter() {defaultValue=0.15, parameterName=GlitchIntensity}, B=Sine(None=@n12) {period=8})
  )
  n14 = TextureSample(UVs=@n13) {texture=&T_VCVFX_Boost_Glitch_Glitch_FB_03_RGB}
  n15 = LinearInterpolate(A=AppendVector(A=AppendVector(A=@n4, B=@n14.G), B=@n14.R), B=AppendVector(A=AppendVector(A=@n14.B, B=@n3.R), B=@n14.G), Alpha=@n1.Highlights)
  n16 = TextureSample(UVs=@n13) {texture=&T_VCVFX_Boost_Glitch_Pixel_Alpha}
  n17 = AppendVector(A=@n16.R, B=TextureSample(UVs=@n13) {texture=&T_VCVFX_Boost_Glitch_Pixel_Alpha}.G)
  n18 = Multiply(A=ScalarParameter() {defaultValue=0.1, parameterName=MaskGlitchSpeed}, B=Time())
  n19 = Panner(Coordinate=TextureCoordinate() {uTiling=0.5}, Time=@n18) {speedX=0.1, speedY=-0.5}
  n20 = Panner(Coordinate=TextureCoordinate() {uTiling=-0.2, vTiling=2.5}, Time=@n18) {speedX=-0.1, speedY=0.1}
  n21 = Panner(Coordinate=TextureCoordinate() {uTiling=3, vTiling=0.25}, Time=@n18) {speedX=-0.5, speedY=0.1}
  n22 = Panner(Coordinate=TextureCoordinate() {vTiling=-0.1}, Time=@n18) {speedX=-0.1, speedY=0.1}
  n23 = Add(
   A=AppendVector(
    A=TextureSample(UVs=@n19) {texture=&T_VCVFX_Boost_Glitch_Pixel_Alpha}.G
    B=TextureSample(UVs=@n20) {texture=&T_VCVFX_Boost_Glitch_Pixel_Alpha}.G
   )
   B=AppendVector(
    A=TextureSample(UVs=@n21) {texture=&T_VCVFX_Boost_Glitch_Pixel_Alpha}.R
    B=TextureSample(UVs=@n22) {texture=&T_VCVFX_Boost_Glitch_Pixel_Alpha}.G
   )
  )
  n24 = TextureSampleParameter2D(UVs=@n23) {parameterName=GlitchTexture, texture=&T_VCVFX_Boost_Glitch_Glitch_FB_03_RGB}
  n25 = TextureCoordinate() {uTiling=8, vTiling=8}
  n26 = Time()
  n27 = LinearInterpolate(
   A=Add(A=ComponentMask(None=@n24.RGB) {g=true, r=true}, B=@n23)
   B=@n25
   Alpha=Multiply(A=ScalarParameter() {defaultValue=0.1, parameterName=MaskGlitchIntensity}, B=Sine(None=@n26) {period=8})
  )
  n28 = Constant() {r=0.25}
  n29 = TextureSample(UVs=@n25) {texture=/VehicleCosmetics/Boosts/Boost_Glitch/Texture/T_VCVFX_Boost_Glitch_SquareMask_RGB}
  n30 = Multiply(
   A=TextureSample(UVs=Panner(Coordinate=@n27, Time=@n28) {speedY=-0.275}) {texture=&T_VCVFX_Boost_Glitch_SquareMask_RGB}.B
   B=Multiply(A=@n29.R, B=@n29.R)
  )
  n31 = Constant() {r=8}
  n32 = Saturate(None=ConstantBiasScale(None=Multiply(A=@n30, B=@n31)) {bias=-0.075, scale=1})
  n33 = Multiply(A=LinearInterpolate(A=@n15, B=AppendVector(A=@n17, B=@n16.B), Alpha=@n1.Distortion), B=@n32)
  n34 = Constant() {r=60}
  n35 = Multiply(A=Saturate(None=Multiply(A=@n33, B=@n34)), B=@n32)
  n36 = LinearInterpolate(A=@n15, B=@n35, Alpha=@n1.InnerDetail)
  n37 = ParticleColor()
  MaterialFunctionCall(Emissive=Multiply(A=Multiply(A=@n1.Brightness, B=@n36), B=@n37.RGB)) {materialFunction=&MF_FX_EmissiveAdjustments}
  n38 = Add(A=@n35, B=Multiply(A=@n32, B=Constant() {r=0.5}))
  Multiply(A=@n37.A, B=Saturate(None=Multiply(A=@n38, B=@n29.R)))
 ]
}
