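# Unlit translucent trail material for the Glitch vehicle boost effect, used on Niagara ribbons and sprites.
# A tiled, time-stepped flipbook and a scrolling RGB mask are UV-distorted by panning pixel textures with sine-pulsed intensity.
# Dynamic parameters balance highlights, detail and brightness; particle color tints emissive, with red or blue triggering near-full desaturation first.
# Opacity combines flipbook luminance, the glitched mask, UV edge fades and particle alpha.
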Material /VehicleCosmetics/Boosts/Boost_Glitch/Material/M_Boost_glitch_Trail
graph {
 properties {
  bUsedWithNiagaraRibbons=true
  bUsedWithNiagaraSprites=true
  blendMode=BLEND_Translucent
  pixelDepthOffsetMode=PDOM_Legacy
  shadingModel=MSM_Unlit
  usedShadingModels=Unlit
 }
 nodes [
  n1 = ComponentMask(None=TextureCoordinate() {vTiling=0}) {r=true}
  n2 = ScalarParameter() {defaultValue=1, parameterName=PatternTileU}
  n3 = ComponentMask(None=Panner(Coordinate=Multiply(A=@n1, B=@n2), Time=ScalarParameter() {parameterName=Pattern_PanU}) {speedX=0.5}) {r=true}
  n4 = ComponentMask(None=TextureCoordinate() {uTiling=0}) {g=true}
  n5 = ScalarParameter() {defaultValue=1, parameterName=PatternTileV}
  n6 = ComponentMask(None=Panner(Coordinate=Multiply(A=@n4, B=@n5), Time=ScalarParameter() {parameterName=Pattern_PanV}) {speedY=0.5}) {g=true}
  n7 = Constant() {r=1}
  n8 = ScalarParameter() {defaultValue=8, parameterName=Columns}
  n9 = Divide(A=@n7, B=@n8)
  n10 = ScalarParameter() {defaultValue=8, parameterName=Rows}
  n11 = Divide(A=@n7, B=@n10)
  n12 = Multiply(A=Time(), B=ScalarParameter() {defaultValue=16, parameterName=FB_Speed})
  n13 = Fmod(A=Floor(None=@n12), B=@n10)
  n14 = Floor(None=Divide(A=@n12, B=@n10))
  n15 = Add(A=Multiply(A=AppendVector(A=@n3, B=@n6), B=AppendVector(A=@n9, B=@n11)), B=AppendVector(A=Divide(A=@n13, B=@n10), B=Divide(A=@n14, B=@n8)))
  n16 = Multiply(A=ScalarParameter() {defaultValue=0.01, parameterName=GlitchSpeed}, B=Sine(None=Time()))
  n17 = Panner(Coordinate=TextureCoordinate() {uTiling=0.25, vTiling=0.25}, Time=@n16) {speedX=-0.5, speedY=0.1}
  n18 = Panner(Coordinate=TextureCoordinate() {uTiling=0.1, vTiling=-0.1}, Time=@n16) {speedX=-0.1, speedY=0.1}
  n19 = Add(
   A=@n15
   B=AppendVector(
    A=TextureSample(UVs=@n17) {texture=/VehicleCosmetics/Boosts/Boost_Glitch/Texture/T_VCVFX_Boost_Glitch_Pixel_Alpha}.R
    B=TextureSample(UVs=@n18) {texture=&T_VCVFX_Boost_Glitch_Pixel_Alpha}.G
   )
  )
  n20 = Multiply(A=ScalarParameter() {defaultValue=0.15, parameterName=GlitchIntensity}, B=Sine(None=Time()) {period=8})
  n21 = TextureSample(UVs=Panner(Coordinate=LinearInterpolate(A=@n15, B=@n19, Alpha=@n20)) {speedX=-0.125}) {texture=/VehicleCosmetics/Boosts/Boost_Glitch/Texture/T_VCVFX_Boost_Glitch_Glitch_FB_03_RGB}
  n22 = DynamicParameter() {paramNames [Highlights, Brightness, InnerDetail, DefaultTextureDetail]}
  n23 = TextureCoordinate()
  n24 = Multiply(A=ScalarParameter() {defaultValue=2, parameterName=MaskGlitchSpeed}, B=Time())
  n25 = TextureSample(UVs=Panner(Coordinate=TextureCoordinate() {uTiling=5, vTiling=0.25}, Time=@n24) {speedX=-0.5, speedY=0.1}) {texture=&T_VCVFX_Boost_Glitch_Pixel_Alpha}
  n26 = TextureSample(UVs=Panner(Coordinate=TextureCoordinate() {vTiling=-0.1}, Time=@n24) {speedX=-0.1, speedY=0.1}) {texture=&T_VCVFX_Boost_Glitch_Pixel_Alpha}
  n27 = TextureSampleParameter2D(UVs=AppendVector(A=@n25.R, B=@n26.G)) {parameterName=GlitchTexture, texture=&T_VCVFX_Boost_Glitch_Glitch_FB_03_RGB}
  n28 = Add(A=@n23, B=ComponentMask(None=@n27.RGB) {g=true, r=true})
  n29 = Multiply(A=ScalarParameter() {defaultValue=0.1, parameterName=MaskGlitchIntensity}, B=Sine(None=Time()) {period=8})
  n30 = TextureSample(UVs=Panner(Coordinate=LinearInterpolate(A=@n28, B=@n23, Alpha=@n29), Time=Constant() {r=0.25}) {speedY=-0.1}) {texture=/VehicleCosmetics/Boosts/Boost_Glitch/Texture/T_VCVFX_Boost_Glitch_RGB_Mask_RGB}
  n31 = Multiply(A=Multiply(A=@n30.R, B=@n27.R), B=Multiply(A=@n25.RGB, B=@n26.RGB))
  n32 = ComponentMask(None=TextureCoordinate()) {g=true}
  n33 = OneMinus(None=@n32)
  n34 = Multiply(A=Saturate(None=ConstantBiasScale(None=@n33) {bias=-0.075, scale=20}), B=Saturate(None=ConstantBiasScale(None=@n32) {bias=-0.075, scale=20}))
  n35 = OneMinus(None=Multiply(A=@n34, B=@n34))
  n36 = Multiply(A=@n21.RGB, B=Add(A=Multiply(A=@n22.InnerDetail, B=@n31), B=Saturate(None=@n35)))
  n37 = Multiply(A=Add(A=Multiply(A=@n36, B=@n22.Highlights), B=Multiply(A=@n21.RGB, B=@n22.DefaultTextureDetail)), B=@n22.Brightness)
  n38 = Desaturation(None=@n37, Fraction=Constant() {r=0.9}) {luminanceFactors=<rgba 0.3 0.59 0.11 0>}
  n39 = ParticleColor()
  n40 = Add(A=@n39.R, B=@n39.R)
  n41 = Saturate(None=Ceil(None=Add(A=@n39.B, B=@n40)))
  n42 = ParticleColor()
  MaterialFunctionCall(Emissive=Multiply(A=LinearInterpolate(A=@n37, B=@n38, Alpha=@n41), B=@n42.RGB)) {materialFunction=&MF_FX_EmissiveAdjustments}
  n43 = Desaturation(None=@n21.RGB) {luminanceFactors=<rgba 0.3 0.59 0.11 1>}
  n44 = Constant() {r=8}
  n45 = Saturate(None=ConstantBiasScale(None=Multiply(A=@n30.R, B=@n44)) {bias=-0.01, scale=1})
  n46 = Saturate(None=Multiply(A=Multiply(A=@n43, B=@n45), B=Constant() {r=60}))
  n47 = Constant() {r=0.5}
  n48 = ComponentMask(None=TextureCoordinate()) {r=true}
  n49 = ConstantBiasScale(None=OneMinus(None=@n48)) {bias=-0.01, scale=2}
  n50 = ConstantBiasScale(None=@n48) {bias=-0.025, scale=20}
  n51 = Multiply(A=Add(A=Multiply(A=@n46, B=@n45), B=Multiply(A=@n45, B=@n47)), B=Multiply(A=Saturate(None=@n49), B=Saturate(None=@n50)))
  Saturate(None=Multiply(A=@n42.A, B=Multiply(A=@n34, B=@n51)))
 ]
}
